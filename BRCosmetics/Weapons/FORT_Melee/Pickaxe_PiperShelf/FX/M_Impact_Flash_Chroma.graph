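# Translucent unlit Niagara sprite material for the chromatic impact flash of the PiperShelf pickaxe.
# A time-panned, polar-mapped rainbow texture, distorted by the impact texture's red and blue channels, multiplies that texture and particle color, boosted early in life, for emissive.
# Particle alpha drives an erosion-style opacity from a separate mask texture.
# A camera-distance offset around the particle center resizes distant sprites; mobile skips it and the color boost.
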Material /BRCosmetics/Weapons/FORT_Melee/Pickaxe_PiperShelf/FX/M_Impact_Flash_Chroma
graph {
 properties {
  bUsedWithNiagaraSprites=true
  blendMode=BLEND_Translucent
  pixelDepthOffsetMode=PDOM_Legacy
  shadingModel=MSM_Unlit
  translucencyLightingMode=TLM_Surface
  usedShadingModels=Unlit
 }
 nodes [
  n1 = VectorParameter() {group=Color, parameterName=Color UV, defaultValue=<rgba 0.5 0.8 -0.01 0>}
  n2 = DynamicParameter() {parameterIndex=1, defaultValue=<rgba 0 0 1 0>, paramNames [UVTilingRandom, TimeOffsetRandom, WPOStrength, TrailingModeON]}
  n3 = Multiply(
   A=MaterialFunctionCall!(UV Coord=TextureCoordinate(), LenghtScale, RadialScale) {materialFunction=/Game/Global/MaterialFunctions/TechArt/MF_PolarCoordinates}.Result
   B=MaterialFunctionCall(X=@n1.R, Y=Add(A=@n2.UVTilingRandom, B=@n1.G)) {materialFunction=&MakeFloat2}.Result
  )
  n4 = Multiply(
   A=MaterialFunctionCall(X=@n1.B, Y=@n1.B) {materialFunction=&MakeFloat2}.Result
   B=Add(A=@n2.TimeOffsetRandom, B=Time())
  )
  n5 = TextureSampleParameter2D() {parameterName=ImpactTexture, texture=&/T_Impact_Flash_2x2_Chromatic}
  n6 = ComponentMask(None=@n5.RGB) {b=true, r=true}
  n7 = TextureSampleParameter2D!(UVs=Add(A=Add(A=@n3, B=@n4), B=Multiply(A=@n6) {constB=0.5}) {constB=0.5}, Level, Apply View MipBias) {group=Color, mipValueMode=TMVM_MipLevel, parameterName=Color Tex, texture=&T_Tile_RainbowColor}
  n8 = Desaturation(None=@n7.RGB, Fraction=ScalarParameter() {group=Color, parameterName=Color Tex Desaturation})
  n9 = ParticleColor()
  n10 = LinearInterpolate(
   A=Multiply(A=ScalarParameter() {defaultValue=4, parameterName="Particle color start  Multiply"}, B=@n9.RGB) {constA=4}
   B=@n9.RGB
   Alpha=ParticleRelativeTime()
  ) {constA=6}
  MaterialFunctionCall(Emissive=Multiply(A=Multiply(A=@n8, B=@n5.RGB), B=ShadingPathSwitch(Default=@n10, Mobile=@n9.RGB))) {materialFunction=&MF_FX_EmissiveAdjustments}
  n11 = TextureSample() {texture=&/T_Impact_Flash_2x2_Chromatic_Masks}
  n12 = Add(A=LinearInterpolate(Alpha=@n9.A) {constA=-0.6, constB=0.6}, B=@n11.G)
  Multiply(A=@n9.A, B=Clamp(None=Multiply(A=@n12, B=@n11.R)))
  n13 = ParticlePositionWS()
  n14 = Subtract(A=@n13, B=WorldPosition().XYZ)
  n15 = ScalarParameter() {defaultValue=5, parameterName=MaxSize}
  n16 = CameraPositionWS()
  n17 = Divide(
   A=Add(A=Distance(A=@n13, B=@n16), B=ScalarParameter() {defaultValue=-1250, parameterName=ScaleDistanceOffset})
   B=ScalarParameter() {defaultValue=20000, parameterName=Scale Distance}
  )
  ShadingPathSwitch(Default=Multiply(A=Multiply(A=@n14, B=@n15), B=Clamp(None=@n17)), Mobile=Constant())
 ]
}
