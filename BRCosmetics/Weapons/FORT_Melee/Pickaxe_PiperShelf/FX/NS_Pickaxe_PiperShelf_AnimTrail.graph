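# Niagara swing-trail effect for the PiperShelf pickaxe cosmetic, layering animation-driven ribbons with sparkles.
# Four ribbon emitters use shared AnimTrail functions on the skeletal mesh for chroma, soft dark and thin tip trails, with alpha scaled by a system FadeOut value.
# A sparkle sprite emitter spawns around a mesh bone or socket in a short burst, drifting with curl noise, drag and gravity.
# All emitters tint color by time of day.
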NiagaraSystem /BRCosmetics/Weapons/FORT_Melee/Pickaxe_PiperShelf/FX/NS_Pickaxe_PiperShelf_AnimTrail
dependencies {
 usedDataInterfaces [&NiagaraDataInterfaceColorCurve, &NiagaraDataInterfaceCurve, &NiagaraDataInterfaceSkeletalMesh]
 usedDynamicInputs []
 usedModules [
  &self:SystemSpawnScript.SystemScriptSource.SystemScriptGraph.NiagaraNodeAssignment_0.SetVariables_#1
  &SystemState
  &self:SystemSpawnScript.SystemScriptSource.SystemScriptGraph.NiagaraNodeAssignment_2.SetVariables_#2
 ]
}
emitters {
 NE_ChromaTrail {
  emitterUpdateScript [
   &EmitterState()
   &FN_SpawnAnimTrail()
  ]
  particleSpawnScript [
   &FN_InitializeAnimTrailBoneIndecies()
   &FN_InitializeAnimTrail(AnimTrailWidth=1.85, Color=<rgba 0.4204 0.5576 0.6406 55>, Lifetime=0.15, SkeletalMesh=$System.Skeletal Mesh)
  ]
  particleUpdateScript [
   &FN_UpdateAnimTrail(SkeletalMesh=$System.Skeletal Mesh) {module=FN_UpdateAnimTrail001}
   &Color(
    Color=&ColorFromCurve(ColorCurve=DataInterface() {propertyValues {
       AlphaCurve {keys [<at 0 1 RCIM_Cubic>,<at 0.4649 0 RCIM_Cubic>]}
       BlueCurve {keys [<at 0 1 RCIM_Cubic>]}
       GreenCurve {keys [<at 0 1 RCIM_Cubic>]}
       RedCurve {keys [<at 0 1 RCIM_Cubic>]}
      }})
    Scale Alpha=$System.FadeOut
   )
   &ScaleColor(
    Color Value To Scale=$Particles.Color
    Scale Mode=::NewEnumerator1
    Scale RGBA=&MakeVector4_FromLinearColor(LinearColor=&FNTimeOfDayColor(
      Night Color=<rgba 0.3333 0.5239 1 1>
      Morning Color=<rgba 1 0.6244 0.9769 1>
      Day Color=<rgba 1 1 1 1>
      Evening Color=<rgba 1 0.6977 0.7077 1>
     ))
    ScaleRGBA=true
   )
   &DynamicMaterialParameters(
    Index 0 Param 1=&RandomRangeFloat(Minimum=0, Maximum=0.2)
    Index 0 Param 2=&RandomRangeFloat(Minimum=0, Maximum=0.2)
    Index 1 Param 2=&FloatFromCurve(FloatCurve=DataInterface() {propertyValues {Curve {keys [<at 0 0 RCIM_Cubic>, {arriveTangent=9.1086, tangentMode=RCTM_User, time=0.5, value=4}]}}})
    Module.Param0WriteEnabled=true
    Module.Param0WriteEnabled1=true
    Module.Param1WriteEnabled=true
    Module.Param1WriteEnabled1=true
    Module.Param2WriteEnabled=true
    Module.Param2WriteEnabled1=true
    Module.Param3WriteEnabled=true
    Module.Param3WriteEnabled1=true
    Write Parameter Index 1=true
   )
  ]
  rendererClasses [&NiagaraRibbonRendererProperties]
  renderers [{rendererClass=&NiagaraRibbonRendererProperties}]
 }
 NE_SoftDarkTrail {
  emitterUpdateScript [
   &EmitterState()
   &FN_SpawnAnimTrail()
  ]
  particleSpawnScript [
   &FN_InitializeAnimTrailBoneIndecies()
   &FN_InitializeAnimTrail(AnimTrailWidth=1.2, Lifetime=0.1, SkeletalMesh=$System.Skeletal Mesh)
  ]
  particleUpdateScript [
   &FN_UpdateAnimTrail(SkeletalMesh=$System.Skeletal Mesh) {module=FN_UpdateAnimTrail001}
   &Color(
    Color=&ColorFromCurve(ColorCurve=DataInterface() {propertyValues {
       AlphaCurve {keys [<at 0 1 RCIM_Cubic>,<at 0.4607 0.25>,<at 1 0 RCIM_Cubic>]}
       BlueCurve {keys [<at 0 0.1146>]}
       GreenCurve {keys [<at 0 0.0961>]}
       RedCurve {keys [<at 0 0.0969>]}
      }})
    Scale Alpha=$System.FadeOut
   )
   &ScaleColor(
    Color Value To Scale=$Particles.Color
    Scale Mode=::NewEnumerator1
    Scale RGBA=&MakeVector4_FromLinearColor(LinearColor=&FNTimeOfDayColor(
      Night Color=<rgba 0.3333 0.5239 1 1>
      Morning Color=<rgba 1 0.5318 0.9822 1>
      Day Color=<rgba 1 1 1 1>
      Evening Color=<rgba 1 0.5938 0.5938 1>
     ))
    ScaleRGBA=true
   ) {module=ScaleColor001}
  ]
  rendererClasses [&NiagaraRibbonRendererProperties]
  renderers [{rendererClass=&NiagaraRibbonRendererProperties}]
 }
 NE_SoftDarkTrailTip {
  emitterUpdateScript [
   &EmitterState()
   &FN_SpawnAnimTrail(AnimTrailWidthMode=::FromSecond)
  ]
  particleSpawnScript [
   &FN_InitializeAnimTrailBoneIndecies()
   &FN_InitializeAnimTrail(AnimTrailWidth=0.1, Lifetime=0.1, SkeletalMesh=$System.Skeletal Mesh)
  ]
  particleUpdateScript [
   &FN_UpdateAnimTrail(SkeletalMesh=$System.Skeletal Mesh) {module=FN_UpdateAnimTrail001}
   &Color(
    Color=&ColorFromCurve(ColorCurve=DataInterface() {propertyValues {
       AlphaCurve {keys [<at 0 0.5 RCIM_Cubic>,<at 1 0 RCIM_Cubic>]}
       BlueCurve {keys [<at 0 0.1562>]}
       GreenCurve {keys [<at 0 0.1211>]}
       RedCurve {keys [<at 0 0.1197>]}
      }})
    Scale Alpha=$System.FadeOut
   )
   &ScaleColor(
    Color Value To Scale=$Particles.Color
    Scale Mode=::NewEnumerator1
    Scale RGBA=&MakeVector4_FromLinearColor(LinearColor=&FNTimeOfDayColor(
      Night Color=<rgba 0.3333 0.5239 1 1>
      Morning Color=<rgba 1 0.5318 0.9822 1>
      Day Color=<rgba 1 1 1 1>
      Evening Color=<rgba 1 0.5938 0.5938 1>
     ))
    ScaleRGBA=true
   ) {module=ScaleColor001}
  ]
  rendererClasses [&NiagaraRibbonRendererProperties]
  renderers [{rendererClass=&NiagaraRibbonRendererProperties}]
 }
 NE_TipTrail {
  emitterUpdateScript [
   &EmitterState()
   &FN_SpawnAnimTrail(AnimTrailWidthMode=::FromSecond)
  ]
  particleSpawnScript [
   &FN_InitializeAnimTrailBoneIndecies()
   &FN_InitializeAnimTrail(AnimTrailWidth=0.1, Lifetime=0.1, SkeletalMesh=$System.Skeletal Mesh)
  ]
  particleUpdateScript [
   &FN_UpdateAnimTrail(SkeletalMesh=$System.Skeletal Mesh) {module=FN_UpdateAnimTrail001}
   &Color(
    Color=&ColorFromCurve(ColorCurve=DataInterface() {propertyValues {
       AlphaCurve {keys [<at 0 1>,<at 0.3291 0 RCIM_Cubic>]}
       BlueCurve {keys [<at 0.0011 3>,<at 0.3569 0.1146>]}
       GreenCurve {keys [<at 0.0011 2.2812>,<at 0.3569 0.0352>]}
       RedCurve {keys [<at 0.0011 2.4725>,<at 0.3569 0.0658>]}
      }})
    Scale Alpha=$System.FadeOut
   )
   &ScaleColor(
    Color Value To Scale=$Particles.Color
    Scale Mode=::NewEnumerator1
    Scale RGBA=&MakeVector4_FromLinearColor(LinearColor=&FNTimeOfDayColor(
      Night Color=<rgba 0.3333 0.5239 1 1>
      Morning Color=<rgba 1 0.5318 0.9822 1>
      Day Color=<rgba 1 1 1 1>
      Evening Color=<rgba 1 0.5938 0.5938 1>
     ))
    ScaleRGBA=true
   ) {module=ScaleColor001}
  ]
  rendererClasses [&NiagaraRibbonRendererProperties]
  renderers [{rendererClass=&NiagaraRibbonRendererProperties}]
 }
 Sparkles {
  emitterUpdateScript [
   &EmitterState(Life Cycle Mode=::NewEnumerator1, Loop Behavior=::NewEnumerator1, Loop Duration=0.2, Scalability Mode=::NewEnumerator1, UseLoopDelay=false)
   &SpawnRate(
    Spawn Probability=&FloatFromCurve(
     FloatCurve=DataInterface() {propertyValues {Curve {keys [
         {tangentMode=RCTM_User, time=0, value=1}
         {arriveTangent=-8.0089, interpMode=RCIM_Cubic, leaveTangent=-8.0089, tangentMode=RCTM_User, time=1, value=0}
        ]}}}
     CurveIndex=$Emitter.NormalizedLoopAge
    )
    SpawnRate=&FloatFromCurve(
     FloatCurve=DataInterface() {propertyValues {Curve {keys [
         {arriveTangent=7.024, interpMode=RCIM_Cubic, leaveTangent=7.024, tangentMode=RCTM_User, time=0, value=0}
         {arriveTangent=4.402, interpMode=RCIM_Cubic, leaveTangent=4.402, tangentMode=RCTM_User, time=0.2, value=1.3}
         {arriveTangent=-13.959, tangentMode=RCTM_User, time=0.4, value=0.875}
        ]}}}
     CurveIndex=$Emitter.Age
     Scale Curve=75
    )
    Use Spawn Probability=false
   ) {module=SpawnRate001}
  ]
  particleSpawnScript [
   &InitializeParticle(
    Lifetime Max=0.5
    Lifetime Min=0.3
    Lifetime Mode=::NewEnumerator1
    Sprite Size Mode=::NewEnumerator1
    Uniform Sprite Size Max=3
    Uniform Sprite Size Min=2
    UsePositionOffset=false
    Write Material Random=true
   ) {module=InitializeParticle001}
   &SkeletalMeshLocation(
    Bone / Socket Index=1
    MaskSpawnGroup=false
    Mesh Sampling Type=::NewEnumerator5
    Skeletal Mesh=$System.Skeletal Mesh
    Socket Sampling Mode=::NewEnumerator5
   ) {module=SkeletalMeshLocation001}
   &ShapeLocation(Box Size=<3 3 20>, MaskSpawnGroup=false, Override Seed=false, Shape Primitive=::NewEnumerator2, Surface Only=false, UseOwnerScale=true) {module=ShapeLocation002}
   &InheritVelocity(Module.Limit Speed=true, Velocity Limit:NiagaraFloat=&RandomRangeFloat(Minimum=15, Maximum=30))
   &AddVelocity(
    Cone Angle=10
    Cone Axis=&RotateVectorByQuaternion(Vector To Rotate=<-1 0 0>, Quaternion=$Particles.SkeletalMeshLocation.SampledRotation)
    Module.UseConeFalloff=false
    Module.UseVelDistribution=false
    Override Seed=false
    Velocity Mode=::NewEnumerator2
    Velocity Speed=&FloatFromCurve(
     FloatCurve=DataInterface() {propertyValues {Curve {keys [
         {arriveTangent=13.4013, interpMode=RCIM_Cubic, leaveTangent=13.4013, tangentMode=RCTM_User, time=0.1, value=0}
         {interpMode=RCIM_Cubic, tangentMode=RCTM_User, time=0.27, value=1}
        ]}}}
     CurveIndex=$Emitter.Age
     Scale Curve=&RandomRangeFloat(Minimum=30, Maximum=60)
    )
   ) {module=AddVelocity001}
  ]
  particleUpdateScript [
   &ParticleState()
   &CurlNoiseForce(
    Noise Strength=&FloatFromCurve(
     FloatCurve=DataInterface() {propertyValues {Curve {keys [{interpMode=RCIM_Cubic, tangentMode=RCTM_User, time=0, value=0.2}, {interpMode=RCIM_Cubic, tangentMode=RCTM_User, time=1, value=1}]}}}
     Scale Curve=150
    )
    Pan Noise=false
    Random Seed=33
   ) {module=CurlNoiseForce001}
   &Drag(Use Linear Drag=true, Use Rotational Drag=false)
   &GravityForce(Gravity=<0 0 -100>)
   &SolveForcesAndVelocity(Clamp Velocity=false, Limit Acceleration=false)
   &CameraOffset(Camera Offset Amount=10)
   &ScaleColor(
    Scale Mode=::NewEnumerator1
    Scale RGBA=&MakeVector4_FromLinearColor(LinearColor=&FNTimeOfDayColor(
      Night Color=<rgba 0.3333 0.5239 1 1>
      Morning Color=<rgba 1 0.6244 0.9769 1>
      Day Color=<rgba 1 1 1 1>
      Evening Color=<rgba 1 0.6977 0.7077 1>
     ))
    ScaleRGBA=true
   ) {module=ScaleColor001}
   &ScaleColor(
    Color Value To Scale=$Particles.Color
    Scale Alpha=&Multiply_Float(A=&FloatFromCurve(FloatCurve=DataInterface() {propertyValues {Curve {keys [<at 0 1>,<at 1 1>]}}}), B=1)
    Scale RGB=&Multiply_Vector(
     A=&Multiply_VectorByFloat(Vector=<1 1 1>, Float=1)
     B=&VectorFromFloat(Value=&RandomRangeFloat(Minimum=0, Maximum=3, Evaluation Type=::NewEnumerator1))
    )
    ScaleA=true
    ScaleRGB=true
   )
   &ScaleSpriteSize(Uniform Curve Sprite Scale=DataInterface() {propertyValues {Curve {keys [
        {interpMode=RCIM_Cubic, tangentMode=RCTM_User, time=0, value=0}
        {interpMode=RCIM_Cubic, tangentMode=RCTM_User, time=0.3, value=4}
        {tangentMode=RCTM_User, time=0.6, value=0}
       ]}}})
   &ScaleSpriteSize(
    Scale Sprite Size Mode=::NewEnumerator0
    Uniform Scale Factor=&RandomRangeFloat(Minimum=0, Maximum=1, Evaluation Type=::NewEnumerator1)
   ) {module=ScaleSpriteSize001}
  ]
  rendererClasses [&NiagaraSpriteRendererProperties]
  renderers [{rendererClass=&NiagaraSpriteRendererProperties}]
 }
}
summary {
 systemName=NS_Pickaxe_PiperShelf_AnimTrail
 emitters [
  {emitterName=NE_TipTrail, rendererClasses [&NiagaraRibbonRendererProperties]}
  {emitterName=NE_SoftDarkTrail, rendererClasses [&NiagaraRibbonRendererProperties]}
  {emitterName=NE_SoftDarkTrailTip, rendererClasses [&NiagaraRibbonRendererProperties]}
  {emitterName=NE_ChromaTrail, rendererClasses [&NiagaraRibbonRendererProperties]}
  {emitterName=Sparkles, rendererClasses [&NiagaraSpriteRendererProperties]}
 ]
 userVariables [User.RibbonIDOffset {type=NiagaraInt32}]
}
