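# Niagara idle effect for the PiperShelf pickaxe: short-lived sparkle sprites placed at points from a hard-coded position array that traces the weapon.
# One emitter spawns them at a quality-scaled rate with a spawn probability.
# Flickering alpha curves are held back by an emitter-age fade-in, then color and alpha are multiplied by values from a shared TimeOfDayGlobal module.
# Three sprite renderers accompany per-particle center, corona and flicker flare parameters.
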NiagaraSystem /BRCosmetics/Weapons/FORT_Melee/Pickaxe_PiperShelf/FX/NS_Pickaxe_PiperShelf_Idle
dependencies {
 usedDataInterfaces [
  &NiagaraDataInterfaceArrayFloat
  &NiagaraDataInterfaceCurve
  &NiagaraDataInterfaceArrayFloat3
  &NiagaraDataInterfaceColorCurve
 ]
 usedDynamicInputs []
 usedModules [&SystemState, /BRCosmetics/Characters/Player/Female/Medium/Bodies/F_MED_PiperShelf_Pearl/FX/TimeOfDayGlobal]
}
emitters {Idle_Sparkles {
  emitterUpdateScript [
   &EmitterState()
   &SpawnRate(
    Spawn Probability=0.125
    SpawnRate=&SelectFloatFromArray(Direct Array Index=$Engine.QualityLevel, Float Selection Array=DataInterface() {propertyValues {FloatData [50,75,100,125,150]}})
    Use Spawn Probability=true
   )
  ]
  particleSpawnScript [
   &InitializeParticle(
    AdjustAlpha=false
    AdjustHue=true
    AdjustSaturation=false
    AdjustValue=false
    Color=<rgba 0.843 0.8 1 1>
    Color Mode=::NewEnumerator3
    Lifetime Max=0.4
    Lifetime Min=0.3
    Lifetime Mode=::NewEnumerator1
    Sprite Size Max=<15 35>
    Sprite Size Min=<10 20>
    Sprite Size Mode=::NewEnumerator2
    UsePositionOffset=true
    Write Material Random=true
   ) {module=InitializeParticle001}
   &CameraOffset(Camera Offset Amount=7)
  ]
  particleUpdateScript [
   &ParticleState()
   &ScaleSpriteSize() {module=ScaleSpriteSize001}
   /SetVariables(Particles.Position=&SelectVectorFromArray(
     Array Sampling Mode=::NewEnumerator0
     Direct Array Index=$Particles.SpawnGroup
     Vector Selection Array=DataInterface() {propertyValues {FloatData [
        <2.3613 0.6286 21.7226>
        <0.9153 -0.0707 20.2429>
        <-0.4649 0.5975 19.1584>
        <-0.9592 -0.1821 18.7702>
        <2.0737 -1.7401 7.011>
        <2.722 0.2761 8.2196>
        <1.1109 2.0895 6.4253>
        <-0.2205 2.2148 5.9705>
        <-0.6808 2.0108 5.313>
        <-2.6264 0.0121 7.1054>
        <4.3323 -0.406 10.0708>
        <2.608 -1.4922 7.8044>
        <5.2222 0.6689 13.1068>
        <3.1238 1.1515 17.9195>
        <1.297 0.6716 18.6679>
        <5.0718 -1.2041 14.5793>
        <-1.2228 -2.4118 8.3693>
        <0.4829 -2.7248 8.6937>
        <4.9766 -1.9044 11.1823>
        <4.4552 -1.9201 9.8712>
        <5.8111 -1.2976 12.0336>
        <5.7316 -1.2733 12.7545>
        <6.0857 -0.9325 13.0532>
        <5.8768 -0.4437 12.675>
        <4.849 -0.6277 10.3148>
        <4.099 -1.6157 9.3326>
        <2.2615 -2.6149 7.9715>
        <0.2974 -2.2849 5.7764>
        <-0.2415 -1.9936 5.1085>
        <-1.9733 -0.6613 4.3817>
        <-0.9505 -2.3139 6.7165>
        <-0.4543 -2.5211 6.6928>
        <-1.2097 -2.3086 7.7019>
        <-2.1294 -1.6608 7.9021>
        <-2.5295 -0.5051 8.1896>
        <-2.2904 0.0524 8.3063>
        <1.5798 1.6336 8.5455>
        <4.0538 0.6239 11.0565>
        <4.2217 1.0069 11.5886>
        <4.6565 1.2059 13.0036>
        <0.9379 2.0727 9.597>
        <-0.1346 -1.6817 25.5707>
        <-0.7021 -1.6274 24.5728>
        <-2.8405 -0.4719 21.6066>
        <-1.2912 -1.9807 22.7419>
        <-3.6832 -1.1123 20.636>
        <-5.2236 -0.5839 18.6735>
        <-5.2318 -1.3655 17.5287>
        <-4.3766 -1.6998 17.2836>
        <-3.6135 -2.0306 18.1917>
        <-1.611 -1.8635 20.8229>
        <-0.4713 -2.1827 20.4176>
        <1.3814 -1.0603 20.6728>
        <0.2821 -1.9575 23.4401>
        <1.4598 -1.5486 26.0207>
        <-1.0569 -0.2223 -16.7747>
        <-1.4668 -0.8969 -14.3988>
        <-1.489 -1.1749 -13.7647>
        <-0.7673 -1.4461 -9.9315>
        <-0.8242 -1.8423 -13.2709>
        <-0.4302 -1.5941 -9.9396>
        <-0.3901 -1.9088 -13.5304>
        <-0.0372 -2.0855 -10.6374>
        <0.0142 -0.5315 -18.8367>
        <1.0706 -2.2521 -11.3403>
        <0.6966 -1.9283 -13.2011>
        <1.5466 -1.6835 -10.9969>
        <1.7604 -0.3172 -14.1213>
        <1.7083 -0.0014 -10.0149>
        <-2.3499 0.06 -12.0945>
        <-2.2077 1.183 -11.4445>
        <-0.5459 0.6444 -17.6609>
        <-0.2033 2.2295 -12.4554>
        <-0.0193 2.1053 -13.0216>
        <0.5488 1.7158 -10.168>
        <0.9212 2.3574 -11.3904>
        <1.2887 1.1704 -14.3062>
        <1.3489 0.7386 -15.06>
        <1.7189 0.2156 -10.0873>
        <0.2578 1.1107 44.9273>
        <0.8728 1.0204 37.0033>
        <-0.304 1.3213 38.8379>
        <0.6601 1.1547 37.26>
        <-0.4616 1.3392 35.7089>
        <0.1166 1.7677 31.0029>
        <0.5916 1.3533 28.5928>
        <-0.444 1.6864 27.7283>
        <0.3532 -0.9578 46.7274>
        <0.1902 -1.125 45.4631>
        <0.6659 -1.0359 40.0816>
        <-0.1697 -1.2968 41.7979>
        <0.7586 -1.171 35.2486>
        <0.6069 -1.4347 31.6386>
        <-0.1217 -1.6217 34.3184>
        <-0.4404 -1.536 30.9173>
        <0.0197 -1.4617 26.3611>
        <2.1181 1.4169 26.467>
        <1.6056 1.3002 28.3973>
        <0.7907 1.969 25.6192>
        <0.8769 1.8611 24.8222>
        <1.0825 1.7805 23.5262>
        <-2.4571 1.5499 21.7579>
        <-3.3082 1.9084 18.8638>
        <-4.3785 1.7943 16.8096>
        <-5.8246 0.1095 15.4762>
        <-5.8596 0.4119 15.3339>
        <-5.5815 0.2589 18.1299>
        <-5.3286 0.5418 18.4826>
       ]}}
     Random Seed=0
    )) {isSetParameters=true}
   &ScaleColor(
    Scale Alpha=&Multiply_Float(
     A=&FloatFromCurve(
      FloatCurve=DataInterface() {propertyValues {Curve {keys [
          {interpMode=RCIM_Cubic, tangentMode=RCTM_User, time=0, value=0}
          {arriveTangent=-2.4058, interpMode=RCIM_Cubic, leaveTangent=-2.4058, tangentMode=RCTM_User, time=0.26, value=0.8421}
          {arriveTangent=-1.1775, interpMode=RCIM_Cubic, leaveTangent=-1.1775, time=0.4613, value=0.0645}
          {arriveTangent=-0.1197, interpMode=RCIM_Cubic, leaveTangent=-0.1197, time=0.5898, value=0.4537}
          <at 1 0 RCIM_Cubic>
         ]}}}
      Scale Curve=&RandomRangeFloat(Minimum=0.1, Maximum=0.5)
     )
     B=&FloatFromCurve(
      FloatCurve=DataInterface() {propertyValues {Curve {keys [{tangentMode=RCTM_User, time=0, value=0}, {tangentMode=RCTM_User, time=1, value=1}]}}}
      Scale Curve=&FloatFromCurve(
       FloatCurve=DataInterface() {propertyValues {Curve {keys [
           {arriveTangentWeight=0.1667, interpMode=RCIM_Cubic, leaveTangentWeight=0.1667, tangentMode=RCTM_User, time=1.4, value=0}
           {tangentMode=RCTM_User, time=2, value=1}
          ]}}}
       CurveIndex=$Emitter.Age
      )
     )
    )
    Scale RGB=&VectorFromFloat(Value=&RandomRangeFloat(Minimum=0.6, Maximum=1, Evaluation Type=::NewEnumerator1))
    ScaleA=true
    ScaleRGB=true
   ) {module=ScaleColor001}
   &ScaleColor(
    Scale Alpha=$System.TODAlpha
    Scale RGB=&MakeVectorFromLinearColorRGB(Color=$System.TODColor)
    ScaleA=true
    ScaleRGB=true
   )
   /SetVariables(
    Particles.CenterFlareColor:LinearColor=&MultiplyLinearColor(
     Linear Color=$Particles.Color
     Scale Factor=&MakeVector4_FromLinearColor(LinearColor=<rgba 8.0099 9.8007 10 1>)
    )
    Particles.CenterFlareSize:Vector2f=&Multiply_Vector2(A=$Particles.SpriteSize, B=&Vector2DFromFloat(Value=0.5))
    Particles.CoronaFlareColor:LinearColor=&MultiplyLinearColor(
     Linear Color=$Particles.Color
     Scale Factor=&MakeVector4_FromLinearColor(LinearColor=<rgba 0.15 0.15 0.15 1>)
    )
    Particles.CoronaFlareSize:Vector2f=&Multiply_Vector2(A=$Particles.Initial.SpriteSize, B=&Vector2DFromFloat(Value=1))
    Particles.FlickerFlareColor:LinearColor=&MultiplyLinearColor(
     Linear Color=$Particles.Color
     Scale Factor=&MakeVector4_FromLinearColor(LinearColor=<rgba 0.9441 0.8079 1 1>)
    )
   ) {isSetParameters=true}
  ]
  rendererClasses [&NiagaraSpriteRendererProperties]
  renderers [
   {rendererClass=&NiagaraSpriteRendererProperties}
   {rendererClass=&NiagaraSpriteRendererProperties, rendererIndex=1}
   {rendererClass=&NiagaraSpriteRendererProperties, rendererIndex=2}
  ]
 }}
summary {
 systemName=NS_Pickaxe_PiperShelf_Idle
 emitters [{emitterName=Idle_Sparkles, rendererClasses [&NiagaraSpriteRendererProperties]}]
 userVariables [
  User.bisSwinging {type=NiagaraBool}
  User.Lifetime_Scale {type=NiagaraFloat, defaultValue=0, description=""}
 ]
}
