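# Physics asset for the PiperShelf two-handed melee pickaxe mesh in the Battle Royale cosmetics content.
# It defines a single body on the Melee_TwoHanded_Handle bone, shaped as one capsule offset and rotated relative to that bone.
# With only one body it has no constraints, so the weapon's collision is a single rigid approximation rather than a jointed chain.
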PhysicsAsset /BRCosmetics/Weapons/FORT_Melee/Pickaxe_PiperShelf/Meshes/PiperShelf_Axe_PhysicsAsset
bodies {Melee_TwoHanded_Handle [{shapeType=Capsule, length=40.7059, radius=6.1678, center=<-0.0044 0.1522 19.7559>, rotation=<rot 0.4849 -90 -0.0449>}]}
constraints []
